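; Season 16 battle pass Agent Jones "Jump" quest actor that shows a hologram mesh when a local player comes within interact range.
; The in-range and out-of-range events run only on the non-authority side, unhiding or hiding SM_Hologram.
; Each first checks that the triggering pawn's controller is among the local player controllers, so visibility is driven per client.
; BeginPlay, overlap and Tick are empty, and the construction script only calls its parent.
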(event EventPlayerOutOfInteractRange
  (params (Actor :type Actor%20Object%20Reference))
  (Networking|SwitchHasAuthority
    (:Remote
      (bind _asfort_player_pawn (Utilities|Casting|CastToFortPlayerPawn Actor))
      (if (Utilities|Array|ContainsItem (Fortnite|GetLocalFortPlayerControllers true) (Fortnite|GetFortPlayerControllerfromActor _asfort_player_pawn))
        (Development|SetHiddeninGame (Variables|Default|GetSM_Hologram) true true)))))

(event EventPlayerInInteractRange
  (params (Actor :type Actor%20Object%20Reference))
  (Networking|SwitchHasAuthority
    (:Remote
      (bind _asfort_player_pawn (Utilities|Casting|CastToFortPlayerPawn Actor))
      (if (Utilities|Array|ContainsItem (Fortnite|GetLocalFortPlayerControllers true) (Fortnite|GetFortPlayerControllerfromActor _asfort_player_pawn))
        (Development|SetHiddeninGame (Variables|Default|GetSM_Hologram) false true)))))

(event EventBeginPlay)

(event Collision|EventActorBeginOverlap
  (params (OtherActor :type Actor%20Object%20Reference)))

(event EventTick
  (params (DeltaSeconds :type Float%20%28single-precision%29)))

(fn ConstructionScript
  (|CallParentFunction))
