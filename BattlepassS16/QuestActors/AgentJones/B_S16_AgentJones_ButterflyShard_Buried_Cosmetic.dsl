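; Cosmetic Blueprint actor for the Season 16 Agent Jones quest that animates a buried butterfly shard emerging, then removes itself.
; BeginPlay stores a start point slightly below the spawn location and plays the Emerge timeline.
; Timeline updates lerp the actor upward from that point, zero its rotation and drive uniform scale; completion destroys the actor.
; Overlap, Tick and construction hold no logic.
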(event EventBeginPlay
  (bind _self self)
  (Variables|Default|SetInitialPosition (+ (Transformation|GetActorLocation) "0, 0,-32.000000"))
  (bind (_direction _alpha _scale) (Timeline|Emerge PlayFromStart)
    (:Update
      (Transformation|SetActorLocationAndRotation _self (Math|Vector|Lerp(Vector) (Variables|Default|GetInitialPosition) (+ (Variables|Default|GetInitialPosition) "0, 0,48.000000") _alpha) "0, 0, 0" false false)
      (Transformation|SetActorScale3D _self (Math|Vector|MakeVector _scale _scale _scale)))
    (:Finished
      (Actor|DestroyActor))))

(event Collision|EventActorBeginOverlap
  (params (OtherActor :type Actor%20Object%20Reference)))

(event EventTick
  (params (DeltaSeconds :type Float%20%28single-precision%29)))

(fn ConstructionScript)
